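# Translucent unlit material for the PulseLance environment smoke trail, drawn as animated cards.
# World position offset cycles cards along a waving trail using time and UV1, orienting them from camera and tangent vectors and interpolating size.
# Two panning pseudo-3D slice-atlas noises distort the normals and round mask, feeding faked sun lighting, backlighting and time-of-day tint.
# Opacity fades at trail ends, by distance and view direction, with depth fade skipped on lower quality levels.
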Material /Game/Effects/Fort_Effects/Effects/Environment/PulseLance/M_SmokeTrail_Cards
graph {
 properties {bAutomaticallySetUsageInEditor=false, blendMode=BLEND_Translucent, shadingModel=MSM_Unlit, usedShadingModels=Unlit}
 nodes [
  n1 = Add(A=NamedRerouteUsage(), B=Multiply(A=LocalPosition().XYZ) {constB=-1})
  n2 = NamedRerouteUsage()
  n3 = CrossProduct(A=@n2, B=NamedRerouteUsage())
  n4 = Normalize(VectorInput=@n3)
  n5 = NamedRerouteUsage()
  n6 = Multiply(A=@n4, B=@n5) {constB=0.25}
  n7 = LocalPosition()
  n8 = ComponentMask(None=@n7.XYZ) {r=true}
  n9 = Multiply(A=Multiply(A=NamedRerouteUsage(), B=VertexColor().R)) {constB=-100}
  n10 = NamedRerouteUsage()
  n11 = CrossProduct(A=@n3, B=@n2)
  n12 = Multiply(A=OneMinus(None=@n5), B=@n11)
  n13 = Multiply(A=CrossProduct(A=@n4, B=@n2), B=@n5) {constB=0.25}
  n14 = ComponentMask(None=@n7.XYZ) {g=true}
  n15 = Add(A=Add(A=Multiply(A=@n6, B=@n8), B=Multiply(A=@n9, B=@n10)), B=Multiply(A=Add(A=@n12, B=@n13), B=@n14))
  n16 = NamedRerouteUsage()
  n17 = Transform(None=Add(A=@n1, B=Multiply(A=@n15, B=@n16))) {transformSourceType=TRANSFORMSOURCE_Local}
  Add(
   A=@n17
   B=MaterialFunctionCall(Offset Amount=ScalarParameter() {parameterName=CameraOffsetAmount}) {materialFunction=/Engine/Functions/Engine_MaterialFunctions02/WorldPositionOffset/CameraOffset}.Clamped Camera Offset
  )
  MaterialFunctionCall(Distance=ScalarParameter() {defaultValue=750, parameterName=Camera Fade}) {materialFunction=&Near_Camera_Fade}
  n18 = VectorParameter() {parameterName=Albedo, defaultValue=<rgba 0.1059 0.1059 0.1059 0>}
  n19 = Normalize(VectorInput=Add(A=NamedRerouteUsage(), B=NamedRerouteUsage()))
  n20 = DotProduct(A=AtmosphericLightVector(), B=@n19)
  n21 = LinearInterpolate(
   A=Multiply(A=@n18.RGB, B=ScalarParameter() {defaultValue=0.25, parameterName=Shadow Scale})
   B=@n18.RGB
   Alpha=Frac(None=Saturate(None=@n20))
  )
  n22 = NamedRerouteUsage()
  n23 = Multiply(
   A=Add(A=Multiply(A=@n21, B=@n22), B=NamedRerouteUsage())
   B=MaterialFunctionCall(
    Value_Morning=ScalarParameter() {defaultValue=1, group=TimeOfDay, parameterName="Value_Morning (S)"}
    Value_Day=ScalarParameter() {defaultValue=1, group=TimeOfDay, parameterName="Value_Day (S)"}
    Value_Evening=ScalarParameter() {defaultValue=1, group=TimeOfDay, parameterName="Value_Evening (S)"}
    Value_Night=ScalarParameter() {defaultValue=1, group=TimeOfDay, parameterName="Value_Night (S)"}
   ) {materialFunction=&Scalar_Time_of_Day}.Result
  )
  n24 = MaterialFunctionCall(
   Color_Morning=VectorParameter() {group=TimeOfDay, parameterName=TOD_Morning, defaultValue=<rgba 1 1 1 1>}.RGB
   Color_Day=VectorParameter() {group=TimeOfDay, parameterName=TOD_Day, defaultValue=<rgba 1 1 1 1>}.RGB
   Color_Evening=VectorParameter() {group=TimeOfDay, parameterName=TOD_Evening, defaultValue=<rgba 1 1 1 1>}.RGB
   Color_Night=VectorParameter() {group=TimeOfDay, parameterName=TOD_Night, defaultValue=<rgba 1 1 1 1>}.RGB
  ) {materialFunction=&Color_Time_of_Day}
  MaterialFunctionCall(Emissive=LinearInterpolate(A=@n23, B=Multiply(A=@n24.Result, B=@n23), Alpha=ScalarParameter() {group=TimeOfDay, parameterName=UseVectorColorTOD?})) {materialFunction=&MF_FX_EmissiveAdjustments}
  n25 = ComponentMask(None=Multiply(A=NamedRerouteUsage(), B=ScalarParameter() {defaultValue=0.35, parameterName=Noise Distortion Amplitude})) {g=true, r=true}
  n26 = TextureCoordinate()
  n27 = Constant() {r=0.5}
  n28 = ScalarParameter() {defaultValue=0.452, parameterName=Sphere Mask Radius}
  n29 = Multiply(
   A=SphereMask(A=Add(A=@n25, B=@n26), B=@n27, Radius=@n28, Hardness=Constant())
   B=SphereMask(A=@n26, B=@n27, Radius=@n28, Hardness=ScalarParameter() {parameterName=Sphere Mask Hardness})
  )
  n30 = ScalarParameter() {defaultValue=0.25, parameterName=Fade in Length, sliderMax=0.5}
  n31 = NamedRerouteUsage()
  n32 = SmoothStep(Min=OneMinus(None=ScalarParameter() {defaultValue=0.25, parameterName=Fade Out Length, sliderMax=0.5}), Value=@n31) {constMin=0.5}
  n33 = Saturate(None=Multiply(A=SmoothStep(Max=@n30, Value=@n31) {constMax=0.5}, B=OneMinus(None=@n32)))
  n34 = MaterialFunctionCall!(
   NearFadeLength=ScalarParameter() {defaultValue=1000, group=CloseFarFader, parameterName="NearFadeLength (S)"}
   NearFadeOffset=ScalarParameter() {defaultValue=300, group=CloseFarFader, parameterName="NearFadeOffset (S)"}
   NearFadeCurve
   NearFadeStartingScale
   NearFadeEndingScale
   FarFadeLength=ScalarParameter() {defaultValue=25000, group=CloseFarFader, parameterName="FarFadeLength (S)"}
   FarFadeOffset=ScalarParameter() {defaultValue=12500, group=CloseFarFader, parameterName="FarFadeOffset (S)"}
   FarFadeCurve
   FarFadeEndingScale
   Opacity Near Distance Bias
   Opacity Far Distance Bias=ScalarParameter() {defaultValue=-2, group=CloseFarFader, parameterName="Opacity Far Distance Bias (S)"}
  ) {materialFunction=/Game/Effects/Fort_Effects/Materials/_Functions/MF_SmoothLod}
  n35 = Multiply(
   A=Multiply(A=Multiply(A=@n29, B=@n33), B=NamedRerouteDeclaration(None=@n34.Opacity) {name=CloseFarOpacity, nodeColor=<rgba 0.2656 0 1 1>})
   B=ScalarParameter() {defaultValue=0.5, parameterName=Opacity Scalar}
  )
  n36 = ScalarParameter() {bUseCustomPrimitiveData=true, defaultValue=1, parameterName=Opacity Prim Scalar}
  n37 = Power(Base=Saturate(None=Multiply(A=@n35, B=@n36)), Exp=ScalarParameter() {defaultValue=1, parameterName=Opacity Exponent})
  n38 = DepthFade(Opacity=@n37, FadeDistance=ScalarParameter() {defaultValue=10, parameterName=Depth Fade Distance, sortPriority=1})
  n39 = Multiply(A=ScalarParameter() {defaultValue=2, parameterName=Opacity Low Scalar}, B=@n37)
  n40 = Subtract(A=CameraPositionWS(), B=ObjectPositionWS())
  n41 = DotProduct(
   A=Normalize(VectorInput=VectorParameter() {parameterName=Direction, defaultValue=<rgba 0.25 0 3 1>}.RGB)
   B=Normalize(VectorInput=Transform(None=@n40) {transformSourceType=TRANSFORMSOURCE_World, transformType=TRANSFORM_Local})
  )
  n42 = ScalarParameter() {defaultValue=0.4, parameterName=Fade Along Direction Begin}
  n43 = ScalarParameter() {defaultValue=0.9, parameterName=Fade Along Direction End}
  n44 = Divide(A=Subtract(A=Abs(None=@n41), B=@n42), B=Max(A=Subtract(A=@n43, B=@n42)) {constB=0.0001})
  Multiply(A=QualitySwitch(Default=@n38, Low=@n39, High=@n38, Medium=@n39, Epic=@n38), B=OneMinus(None=Saturate(None=@n44)))
  n45 = Transform(None=WorldPosition() {worldPositionShaderOffset=WPT_CameraRelative}.XYZ) {transformSourceType=TRANSFORMSOURCE_World, transformType=TRANSFORM_Local}
  NamedRerouteDeclaration(None=Normalize(VectorInput=Add(A=@n1, B=@n45))) {name=Card Vector, nodeColor=<rgba 1 0.6797 0 1>}
  n46 = NamedRerouteUsage()
  n47 = NamedRerouteUsage()
  n48 = ScalarParameter() {parameterName=Waving Timescale}
  n49 = VectorParameter() {parameterName=Waving Direction}
  n50 = Add(
   A=AppendVector(A=Multiply(A=@n46) {constB=0.6}, B=Constant())
   B=Multiply(A=Multiply(A=@n47, B=@n48), B=ComponentMask(None=@n49.RGB) {g=true, r=true}) {constB=-0.15}
  )
  n51 = MaterialFunctionCall!(
   UVs=@n50
   Offset=Constant() {r=0.02}
   Linear Mask=NamedRerouteUsage()
   Non Linear Mask=NamedRerouteUsage()
   Tangent Multiplier=Constant() {r=0.5}
  ) {materialFunction=/Game/Environments/TechArt/SmokeTrail/MF_SmokeTrail_Waving}
  n52 = MaterialFunctionCall(
   UVs=@n50
   Offset=Constant() {r=-0.02}
   Linear Mask=NamedRerouteUsage()
   Non Linear Mask=NamedRerouteUsage()
   Tangent Multiplier=Constant() {r=0.5}
  ) {materialFunction=&MF_SmokeTrail_Waving}
  NamedRerouteDeclaration(None=Normalize(VectorInput=Subtract(A=@n51.Result, B=@n52.Result))) {name=Waving Tangent, nodeColor=<rgba 0 1 0.8594 1>}
  n53 = ScalarParameter() {parameterName=Backlight Intensity Low}
  n54 = AtmosphericLightVector()
  n55 = DotProduct(A=Multiply(A=@n54) {constB=-1}, B=CameraVectorWS())
  n56 = NamedRerouteUsage()
  n57 = Saturate(None=DotProduct(A=@n54, B=Multiply(A=@n56) {constB=-1}))
  n58 = ScalarParameter() {defaultValue=2, parameterName=Directionality Power}
  n59 = Multiply(
   A=Power(Base=Saturate(None=@n55), Exp=ScalarParameter() {defaultValue=4, parameterName=Backlight Scattering Exponent})
   B=OneMinus(None=Power(Base=@n57, Exp=@n58))
  )
  NamedRerouteDeclaration(None=Multiply(A=NamedRerouteUsage(), B=LinearInterpolate(A=@n53, Alpha=@n59))) {name=Backlighting, nodeColor=<rgba 0 1 0.5078 1>}
  n60 = CollectionParameter() {collection=&FortniteMaterialParameters, parameterName=SunLightColor}
  NamedRerouteDeclaration(None=Desaturation(None=ComponentMask(None=@n60) {b=true, g=true, r=true}, Fraction=ScalarParameter() {defaultValue=-0.5, parameterName=Desaturate Sunlight})) {name=Sunlight Color, nodeColor=<rgba 0 0.2031 1 1>}
  n61 = VertexInterpolator(VS=LocalPosition() {localOrigin=InstancePreSkinning}.XYZ)
  n62 = Multiply(A=NamedRerouteUsage(), B=VectorParameter() {parameterName=Noise Pan Dir, defaultValue=<rgba 0 0 -25 0>}.RGB)
  n63 = ActorPositionWS()
  n64 = Add(A=Add(A=Add(A=@n61.PS, B=@n62), B=NamedRerouteUsage()), B=@n63)
  n65 = ScalarParameter() {defaultValue=1024, parameterName=Noise Scale}
  n66 = Multiply(A=NamedRerouteUsage(), B=VectorParameter() {parameterName=Noise Pan Dir Secondary, defaultValue=<rgba 0 0 35 0>}.RGB)
  n67 = Add(A=Add(A=Add(A=@n61.PS, B=@n66), B=NamedRerouteUsage()), B=@n63)
  n68 = ScalarParameter() {defaultValue=512, parameterName=Noise Scale Secondary}
  n69 = LinearInterpolate(
   A=MaterialFunctionCall!(
    UVs=Divide(A=@n64, B=@n65) {constB=200}
    Slices Texture=TextureObject() {samplerType=SAMPLERTYPE_LinearColor, texture=/Game/Environments/Helios/TechArt/LowAltitudeClouds/T_LowAltitudeClouds_N}
    Atlas Size=Constant() {r=8}
    Atlas Padding=ScalarParameter() {defaultValue=0.9375, parameterName=AtlasPadding, primitiveDataIndex=8}
   ) {materialFunction=/Game/Environments/Helios/TechArt/LowAltitudeClouds/MF_3DVolumeFrom2DSlices}.RGB
   B=MaterialFunctionCall(
    UVs=Divide(A=@n67, B=@n68) {constB=200}
    Slices Texture=TextureObject() {samplerType=SAMPLERTYPE_LinearColor, texture=&T_LowAltitudeClouds_N}
    Atlas Size=Constant() {r=8}
    Atlas Padding=ScalarParameter() {defaultValue=0.9375, parameterName=AtlasPadding, primitiveDataIndex=8}
   ) {materialFunction=&MF_3DVolumeFrom2DSlices}.RGB
  )
  NamedRerouteDeclaration(None=Multiply(A=Add(A=@n69) {constB=-0.5}) {constB=2}) {name=Noise Dir, nodeColor=<rgba 0.8516 1 0 1>}
  MaterialExpressionActorPositionWS_0 = ActorPositionWS()
  n70 = Convert(None=@MaterialExpressionActorPositionWS_0) {
   nodeName=Break Float3
   convertInputs [{
     type=Vector3
     expressionInput {
      expression=&self:MaterialExpressionActorPositionWS_0
      inputName=None
      mask=0
      maskA=0
      maskB=0
      maskG=0
      maskR=0
      outputIndex=0
     }
    }]
   convertOutputs [{type=Scalar}, {type=Scalar}, {type=Scalar}]
  }
  n71 = Add(A=Add(A=@n70, B=@n70), B=@n70)
  n72 = ScalarParameter() {parameterName=seed}
  NamedRerouteDeclaration(None=Add(A=Time(), B=Multiply(A=@n71, B=@n72))) {name=UniqueTimeOffset, nodeColor=<rgba 0.7109 1 0 1>}
  n73 = TextureCoordinate() {coordinateIndex=1}
  n74 = NamedRerouteUsage()
  n75 = ScalarParameter() {defaultValue=0.05, parameterName=Overall Timescale}
  n76 = Frac(None=Add(A=ComponentMask(None=@n73) {r=true}, B=Multiply(A=@n74, B=@n75) {constB=0.1}))
  NamedRerouteDeclaration(None=Power(
    Base=NamedRerouteDeclaration(None=@n76) {name=Linear Mask, nodeColor=<rgba 0.9453 0 1 1>}
    Exp=ScalarParameter() {defaultValue=1.6, parameterName=Shape Exponent}
   ) {constExponent=1.6}) {name=Non Linear Mask, nodeColor=<rgba 0.1016 1 0 1>}
  NamedRerouteDeclaration(None=LinearInterpolate(
    A=ScalarParameter() {defaultValue=0.5, parameterName=Width Start}
    B=ScalarParameter() {defaultValue=1, parameterName=Width End}
    Alpha=NamedRerouteUsage()
   )) {name=Card Width, nodeColor=<rgba 1 0 0.4219 1>}
  n77 = ScalarParameter() {defaultValue=0.4, parameterName=Size Start}
  n78 = ScalarParameter() {defaultValue=2, parameterName=Size End}
  n79 = NamedRerouteUsage()
  NamedRerouteDeclaration(None=Multiply(A=LinearInterpolate(A=@n77, B=@n78, Alpha=@n79), B=NamedRerouteUsage())) {name=Card Size, nodeColor=<rgba 0 0.7422 1 1>}
  NamedRerouteDeclaration(None=@n17) {name=WPO, nodeColor=<rgba 0.4297 0 1 1>}
  NamedRerouteDeclaration(None=MaterialFunctionCall(UVs=@n50, Offset=Constant(), Linear Mask=NamedRerouteUsage(), Non Linear Mask=NamedRerouteUsage(), Tangent Multiplier=Constant() {r=1}) {materialFunction=&MF_SmokeTrail_Waving}.Result) {name=Waving WPO, nodeColor=<rgba 1 0 0.7969 1>}
  n80 = Multiply(A=@n11, B=@n14)
  n81 = Transform(None=Multiply(A=Subtract(A=@n15, B=@n80)) {constB=0.01}) {transformSourceType=TRANSFORMSOURCE_Local}
  n82 = MaterialFunctionCall() {materialFunction=&ObjectScale}
  NamedRerouteDeclaration(None=VertexInterpolator(VS=Divide(A=@n81, B=@n82.Scale XYZ)).PS) {name=Normals, nodeColor=<rgba 0.4062 0 1 1>}
  NamedRerouteDeclaration(None=@n19) {name=Distorted Normal, nodeColor=<rgba 0.5703 1 0 1>}
  NamedRerouteUsage()
  NamedRerouteUsage()
 ]
}
